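# Translucent, unlit shard material for the Star Walker glider, drawing glassy galaxy pieces that open outward and drift.
# World position offset turns shards around the actor and pushes them along directions stored in vertex color and UV2, driven by open parameters and sine vibration.
# Color blends tinted cubemap Fresnel and glowing cracks into a time-of-day-scaled curve-atlas color by radial distance.
# Opacity combines a UV gradient with a radial edge fade.
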Material /BRCosmetics/Athena/Effects/Materials/Gliders/Star_Walker/M_Galaxy_Shards_02
graph {
 properties {
  bUsedWithMeshDeformer=true
  bUsedWithSkeletalMesh=true
  blendMode=BLEND_Translucent
  ditherOpacityMask=true
  pixelDepthOffsetMode=PDOM_Legacy
  refractionMethod=RM_IndexOfRefraction
  shadingModel=MSM_Unlit
  twoSided=true
  usedShadingModels=Unlit
 }
 nodes [
  n1 = VertexColor()
  n2 = Sine(None=Add(A=Time(), B=LinearInterpolate(Alpha=@n1.R) {constA=-1, constB=2})) {period=5.8}
  n3 = ScalarParameter() {defaultValue=0.025, parameterName=TurnAmount}
  n4 = Add(A=@n3, B=Multiply(A=@n3, B=LinearInterpolate(Alpha=@n1.R) {constA=0.3})) {constB=0.05}
  n5 = VertexColor()
  n6 = TextureCoordinate() {coordinateIndex=2}
  n7 = ComponentMask(None=@n6) {r=true}
  n8 = ComponentMask(None=@n6) {g=true}
  n9 = Subtract(A=AppendVector(A=OneMinus(None=@n5.G), B=AppendVector(A=@n7, B=@n8))) {constB=0.5}
  n10 = Multiply(
   A=Transform(None=Multiply(A=@n9) {constB=-2}) {transformSourceType=TRANSFORMSOURCE_Local}
   B=ScalarParameter() {defaultValue=38, parameterName=OpenDistance}
  )
  n11 = LinearInterpolate(Alpha=@n1.R) {constA=0.25, constB=0.05}
  n12 = Time()
  n13 = Multiply(A=@n12) {constB=0.25}
  n14 = MaterialFunctionCall() {materialFunction=&DebugTimeSine}
  n15 = Sine(None=@n12) {period=1.32}
  n16 = ScalarParameter() {parameterName=OpenAdd}
  n17 = Multiply(
   A=Multiply(
    A=MaterialFunctionCall!(Time=@n13) {materialFunction=/Game/Effects/Fort_Effects/Materials/Geode/_Functions/MF_SineVibration}.Result
    B=Add(A=@n14.Result, B=@n15)
   )
   B=Multiply(A=Add(A=@n16)) {constB=0.03}
  ) {constB=0.03}
  n18 = Add(
   A=Multiply(
    A=MaterialFunctionCall(Speed=@n11) {materialFunction=&DebugTimeSine}.Result
    B=Add(A=@n17) {constB=0.18}
   ) {constB=-0.17}
   B=Add(A=ScalarParameter() {parameterName=Open}, B=@n16)
  )
  Add(
   A=RotateAboutAxis(
    NormalizedRotationAxis=Constant3Vector() {constant=<rgba 0 0 1 0>}
    RotationAngle=Multiply(A=@n2, B=@n4) {constB=0.1}
    PivotPoint=ActorPositionWS()
    Position=WorldPosition().XYZ
   )
   B=Multiply(A=Multiply(A=@n10, B=@n18), B=ScalarParameter() {defaultValue=1, parameterName=CloseMult, sliderMax=1})
  )
  n19 = Add(A=TextureCoordinate() {coordinateIndex=2, uTiling=1.25, vTiling=0.5}, B=Multiply(A=TextureCoordinate() {coordinateIndex=1}) {constB=0.5})
  Add(
   A=Multiply(
    A=TextureSampleParameter2D() {parameterName=Normals, samplerType=SAMPLERTYPE_Normal, texture=/Game/Effects/Fort_Effects/Textures/Tile/T_Glass_Cracks_N}.RGB
    B=Constant3Vector() {constant=<rgba 2 2 1 0>}
   )
   B=Multiply(
    A=TextureSampleParameter2D(UVs=@n19) {parameterName=NormalsTile, samplerType=SAMPLERTYPE_Normal, texture=&T_Glass_Cracks_N}.RGB
    B=Constant3Vector() {constant=<rgba 1 1 0 0>}
   )
  )
  Constant() {r=1}
  n20 = TextureSampleParameterCube!(UVs=ReflectionVectorWS() {bNormalizeCustomWorldNormal=true}, Apply View MipBias) {parameterName=Cubemap, samplerType=SAMPLERTYPE_LinearColor, texture=/Game/Packages/Fortress_Sky/TexturesHDR/T_AthenaSkylight}
  n21 = Multiply(A=@n20.RGB, B=VectorParameter() {parameterName=CubemapTint, defaultValue=<rgba 0.4197 0.7965 1 1>}.RGB) {constB=0.7}
  n22 = Fresnel() {baseReflectFraction=0.3, exponent=3}
  n23 = LinearInterpolate(Alpha=Fresnel() {baseReflectFraction=0, exponent=3}) {constA=1, constB=2}
  n24 = ScalarParameter() {defaultValue=1, parameterName=Opacity}
  n25 = LinearInterpolate(A=Constant2Vector() {g=0.25}, Alpha=@n24) {constA=0.5, constB=0}
  n26 = Multiply(A=TextureSample(UVs=Add(A=@n25, B=@n19)) {texture=/Game/Effects/Fort_Effects/Textures/Tile/T_Glass_Cracks_M}.G) {constB=0.6}
  n27 = TextureSample(UVs=TextureCoordinate()) {texture=&T_Glass_Cracks_M}
  n28 = VectorParameter() {parameterName=CrackColor, defaultValue=<rgba 0.2058 0.778 1 0>}
  n29 = ComponentMask(None=TextureCoordinate() {coordinateIndex=2}) {g=true}
  n30 = Power(Base=@n29)
  n31 = Add(
   A=Multiply(A=Add(A=@n26, B=@n27.G), B=Multiply(A=@n28.RGB) {constB=8})
   B=LinearInterpolate(A=VectorParameter() {parameterName=BaseColor, defaultValue=<rgba 0.1405 0.1916 0.5 0>}.RGB, B=@n28.RGB, Alpha=@n30)
  )
  n32 = Subtract(A=ObjectPositionWS(), B=WorldPosition().XYZ)
  n33 = Transform(None=VectorParameter() {group=Color Shift, parameterName=Color Shift Offset}.RGB) {transformSourceType=TRANSFORMSOURCE_Local}
  n34 = ScalarParameter() {defaultValue=3, group=Color Shift, parameterName=Color Shift Edge Range}
  n35 = ObjectRadius()
  n36 = Divide(
   A=MaterialFunctionCall(Vector 3=Add(A=@n32, B=@n33)) {materialFunction=/Engine/Functions/Engine_MaterialFunctions02/Utility/VectorLength}.V3 Length
   B=StaticSwitchParameter(True=Multiply(A=@n34, B=@n35), False=@n34) {defaultValue=true, parameterName=Use Object Radius For Falloff?}
  )
  n37 = ScalarParameter() {defaultValue=5, group=Color Shift, parameterName=Color Shift Edge Power}
  n38 = Clamp(None=Multiply(A=Power(Base=@n36, Exp=@n37), B=ScalarParameter() {defaultValue=500, group=Color Shift, parameterName=Color Shift Edge Hardness}))
  n39 = CurveAtlasRowParameter(CurveTime=@n38) {
   atlas=&AthenaCurveAtlas01
   curve=/Game/Gadgets/Assets/VinderTech_GliderChute/Glider_ShatterFly/FX/Curve_Color
   defaultValue=74
  }
  n40 = Multiply(A=@n39, B=ScalarParameter() {defaultValue=4, group=Color Shift, parameterName=Secondary Color Multiplier})
  n41 = Constant() {r=0.8}
  n42 = MaterialFunctionCall(Color_Morning=@n41, Color_Day=Constant() {r=1}, Color_Evening=@n41, Color_Night=Constant() {r=0.6}) {materialFunction=&Color_Time_of_Day}
  LinearInterpolate(A=Add(A=Multiply(A=@n21, B=@n22), B=Multiply(A=@n23, B=@n31)), B=Multiply(A=Multiply(A=@n40, B=@n42.Result), B=@n20.RGB), Alpha=@n38)
  MaterialFunctionCall() {materialFunction=&DebugTimeSine}
  n43 = Multiply(A=@n24, B=Power(Base=@n29) {constExponent=10})
  n44 = ObjectPositionWS()
  n45 = WorldPosition()
  n46 = VectorParameter() {group=Fade, parameterName=Fade Mask Offset}
  n47 = MaterialFunctionCall(Vector 3=Add(A=Subtract(A=@n44, B=@n45.XYZ), B=Transform(None=@n46.RGB) {transformSourceType=TRANSFORMSOURCE_Local})) {materialFunction=&VectorLength}
  n48 = ScalarParameter() {defaultValue=3, group=Fade, parameterName=Fade Edge Range}
  n49 = StaticSwitchParameter(True=Multiply(A=@n48, B=ObjectRadius()), False=@n48) {defaultValue=true, parameterName=Use Object Radius For Falloff?}
  n50 = Multiply(
   A=Power(Base=Divide(A=@n47.V3 Length, B=@n49), Exp=ScalarParameter() {defaultValue=5, group=Fade, parameterName=Fade Edge Power})
   B=ScalarParameter() {defaultValue=500, group=Fade, parameterName=Fade Edge Hardness}
  )
  Multiply(
   A=Multiply(A=LinearInterpolate(Alpha=@n30) {constA=0.2, constB=1.05}, B=LinearInterpolate(A=@n24, Alpha=@n43)) {constB=1.5}
   B=OneMinus(None=Clamp(None=@n50))
  )
 ]
}
